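# Material function for the ranked umbrella glider that adds an animated swirling noise glow to Emissive, behind the UseRankedSwirlingNoise static switch.
# Pre-skinned local position becomes radial UVs, where scrolling Perlin noise warps a second, radius-twisted sample of the same texture.
# The noise is faded and masked near the center, tinted by EmissiveColor and EmissiveStrength, then lerped in by the Alpha input.
# Both samples take derivatives from mesh UVs rather than the radial UVs.
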MaterialFunction /BRCosmetics/Gadgets/Assets/VinderTech_Umbrella/Umbrella_RankedRube/Materials/MF_RankedGlider_SwirlingNoise
graph {nodes [
  n1 = GetMaterialAttributes(None=FunctionInput() {inputName=Material, inputType=FunctionInput_MaterialAttributes}) {attributeGetTypes [#1]}
  n2 = VectorParameter() {group=EmissiveRankEffect, parameterName=EmissiveColor, defaultValue=<rgba 0.6875 0.1451 0.0903 1>}
  n3 = LinearInterpolate(A=NamedRerouteUsage(), Alpha=NamedRerouteUsage()) {constB=0}
  n4 = LinearInterpolate(
   A=@n1.EmissiveColor
   B=Multiply(A=Multiply(A=@n2.RGB, B=@n3), B=ScalarParameter() {defaultValue=4, group=EmissiveRankEffect, parameterName=EmissiveStrength}) {constB=7}
   Alpha=FunctionInput() {inputName=Alpha, inputType=FunctionInput_Scalar, sortPriority=1}
  )
  FunctionOutput(None=StaticSwitchParameter(
    True=SetMaterialAttributes!(MaterialAttributes=@n1.MaterialAttributes, EmissiveColor=@n4) {attributeSetTypes [#1]}
    False=@n1.MaterialAttributes
   ) {group=EmissiveRankEffect, parameterName=UseRankedSwirlingNoise})
  n5 = LocalPosition() {localOrigin=InstancePreSkinning}
  NamedRerouteDeclaration(None=FunctionInput(Preview=VertexInterpolator(VS=@n5.XYZ).PS) {inputName=PreskinLocPos}) {name=PS-LocPos, nodeColor=<rgba 1 0.3281 0 1>}
  n6 = Multiply(A=Time(), B=Constant2Vector() {desc=SPEED, g=0.1678, r=0.1})
  n7 = NamedRerouteUsage()
  n8 = MaterialFunctionCall(Vector or UVs=Divide(
    A=ComponentMask(None=@n7) {g=true, r=true}
    B=ScalarParameter() {defaultValue=300, group=EmissiveRankEffect_SwirlingNoise, parameterName=GliderSwirlingNoiseSize}
   ) {constB=300}) {materialFunction=&VectorToRadialValue}
  n9 = Multiply(A=@n8.Radial Coordinates, B=Constant2Vector() {g=-2, r=2})
  n10 = TextureCoordinate()
  n11 = DDX(Value=@n10)
  n12 = DDY(Value=@n10)
  n13 = TextureSample!(
   UVs=Add(A=Frac(None=@n6), B=NamedRerouteDeclaration(None=@n9) {name=RadialUV, nodeColor=<rgba 0.3125 1 0 1>}) {constB=0}
   Tex
   "DDX(UVs)"=@n11
   "DDY(UVs)"=@n12
   Apply View MipBias
  ) {mipValueMode=TMVM_Derivative, samplerSource=SSM_Wrap_WorldGroupSettings, texture=/Game/Global/Textures/Default/Noise/3xPerlin_M}
  n14 = ComponentMask(None=@n13.RGB) {g=true, r=true}
  n15 = ScalarParameter() {defaultValue=0.4, group=EmissiveRankEffect_SwirlingNoise, parameterName=WarpingIntensity}
  n16 = VectorParameter() {desc=sPEED, group=EmissiveRankEffect_SwirlingNoise, parameterName=SwirlSpeed, defaultValue=<rgba 0.2 0.15 0 1>}
  n17 = Multiply(A=Time(), B=ComponentMask(None=@n16.RGB) {g=true, r=true}) {constA=50}
  n18 = NamedRerouteDeclaration(None=ComponentMask(None=@n8.Radial Coordinates) {g=true}) {name=CenterGradient, nodeColor=<rgba 1 0.2344 0 1>}
  n19 = LinearInterpolate(
   B=ScalarParameter() {defaultValue=0.75, group=EmissiveRankEffect_SwirlingNoise, parameterName=SwirlIntensity}
   Alpha=Saturate(None=Multiply(A=@n18) {constB=3})
  ) {constB=-0.75}
  n20 = Constant()
  n21 = Add(
   A=NamedRerouteUsage()
   B=Add(
    A=Frac(None=@n17)
    B=MaterialFunctionCall(X=@n19, Y=@n20) {materialFunction=&MakeFloat2}.Result
   )
  ) {constB=0}
  n22 = TextureSample(UVs=Add(A=Multiply(A=@n14, B=@n15), B=Frac(None=@n21)), "DDX(UVs)"=@n11, "DDY(UVs)"=@n12) {mipValueMode=TMVM_Derivative, samplerSource=SSM_Wrap_WorldGroupSettings, texture=&3xPerlin_M}
  n23 = Multiply(A=@n22.R, B=SmoothStep(Value=NamedRerouteUsage()) {constMax=0.3, constMin=0.08})
  n24 = ScalarParameter() {defaultValue=2, group=EmissiveRankEffect_SwirlingNoise, parameterName=SwirlingNoisePower}
  NamedRerouteDeclaration(None=Multiply(
    A=Power(Base=@n23, Exp=@n24)
    B=ScalarParameter() {defaultValue=0.3, group=EmissiveRankEffect_SwirlingNoise, parameterName=SwirlingNoiseIntensity}
   )) {name=SwirlingNoise, nodeColor=<rgba 0 1 0.0859 1>}
  NamedRerouteDeclaration(None=Step(Y=@n18, X=ScalarParameter() {defaultValue=0.09, group=EmissiveRankEffect_SwirlingNoise, parameterName=CenterMask_Position}) {constX=0.05, constY=0.5}) {name=CenterMask, nodeColor=<rgba 0.8047 1 0 1>}
 ]}
