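# Parent material for the Ranked Rube VinderTech umbrella glider cosmetic, built on the 2023 character parent PreFX and PostFX functions.
# Between those functions it chains ranked emissive fresnel, thin film, Defender icon and swirling noise layers.
# A UseRankedRubeGlider static switch overrides base color and emissive using masks and a curve-atlas color driven by a sine of radial coordinates.
# Stepped radial and axis values also build circle, symbol, two-tone gem and swirl masks.
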Material /BRCosmetics/Gadgets/Assets/VinderTech_Umbrella/Umbrella_RankedRube/Materials/M_RankedRube_Parent
graph {
 properties {
  bUseMaterialAttributes=true
  bUsedWithMeshDeformer=true
  bUsedWithNanite=true
  bUsedWithSkeletalMesh=true
  pixelDepthOffsetMode=PDOM_Legacy
  shadingModel=MSM_FromMaterialExpression
  usedShadingModels=Default Lit | Subsurface | Clear Coat | Subsurface Profile | Cloth
 }
 nodes [
  n1 = MaterialFunctionCall() {materialFunction=&MF_CharacterParent_2023_PreFX}
  n2 = NamedRerouteDeclaration(None=@n1.Fresnel) {name=Fresnel, nodeColor=<rgba 1 0.7734 0 1>}
  n3 = NamedRerouteUsage()
  n4 = MaterialFunctionCall!(
   Material Attributes=MaterialFunctionCall!(Material=NamedRerouteUsage(), Mask=NamedRerouteUsage(), Fresnel=@n2) {materialFunction=/Game/Global/Materials/Cosmetics/Ranked/MF_RankedEmissiveFresnel}.Result
   Alpha=@n3
  ) {materialFunction=/Game/Global/Materials/Cosmetics/Ranked/MF_RankedThinFilm}
  MaterialFunctionCall(
   Material=MaterialFunctionCall!(
    Material=MaterialFunctionCall!(Material=@n4.Result, Alpha=@n3) {materialFunction=/BRCosmetics/Accessories/FORT_Backpacks/Backpack_M_MED_RankedDefender/Material/MF_RankedDefenderIcon}.Result
    PreskinLocPos=NamedRerouteUsage()
    Alpha=NamedRerouteUsage()
   ) {materialFunction=&/MF_RankedGlider_SwirlingNoise}.Result
   M=@n1.M
   Pre Skinned Local Position=@n1.M
   Distance Blend=@n1.M
   Fresnel=@n2
  ) {materialFunction=&MF_CharacterParent_2023_PostFX}
  NamedRerouteDeclaration(None=@n1.Result) {name=MA_IN, nodeColor=<rgba 0.5234 1 0 1>}
  n5 = NamedRerouteUsage()
  n6 = GetMaterialAttributes(None=@n5) {attributeGetTypes [#1, #2]}
  n7 = NamedRerouteUsage()
  n8 = NamedRerouteUsage()
  NamedRerouteDeclaration(None=StaticSwitchParameter(
    True=SetMaterialAttributes!(MaterialAttributes=@n6.MaterialAttributes, Emissive Color=@n7, Base Color=@n8) {attributeSetTypes [#1, #2]}
    False=@n5
   ) {group=RankedRube, parameterName=UseRankedRubeGlider}) {name=MA_OUT, nodeColor=<rgba 1 0 0.8437 1>}
  n9 = NamedRerouteUsage()
  n10 = MaterialFunctionCall(Vector or UVs=Divide(A=ComponentMask(None=@n9) {g=true, r=true}, B=Constant() {r=133.5}) {constB=64}) {materialFunction=&VectorToRadialValue}
  n11 = ComponentMask(None=@n10.Radial Coordinates) {g=true}
  NamedRerouteDeclaration(None=Step(X=@n11) {constY=0.3}) {name=GradientCircleMask, nodeColor=<rgba 1 0.9141 0 1>}
  n12 = ComponentMask(None=@n10.Radial Coordinates) {r=true}
  n13 = Multiply(A=@n12) {constB=4}
  n14 = Add(A=Sine(None=Add(A=@n13) {constB=0.25})) {constB=0.25}
  NamedRerouteDeclaration(None=LinearInterpolate(A=Constant3Vector() {constant=<rgba 1 0.1775 0 1>}, B=Constant3Vector() {constant=<rgba 0.8023 0 1 1>}, Alpha=Step(X=@n14))) {name=GemsColor, nodeColor=<rgba 0.5938 0 1 1>}
  n15 = NamedRerouteUsage()
  NamedRerouteDeclaration(None=Step(X=ComponentMask(None=@n15) {b=true}) {constY=115}) {name=SymbolMask, nodeColor=<rgba 1 0 0.9375 1>}
  n16 = NamedRerouteUsage()
  n17 = SmoothStep(Value=ComponentMask(None=@n16) {g=true}) {constMin=0.5}
  n18 = ScalarParameter() {defaultValue=1, group=RankedRube, parameterName=RankedBasecolorIntensity}
  NamedRerouteDeclaration(None=LinearInterpolate(A=LinearInterpolate(A=@n6.Base Color, Alpha=@n17) {constB=0}, B=Multiply(A=@n6.Base Color, B=@n18), Alpha=ComponentMask(None=@n16) {b=true})) {name=BaseColor_OUT, nodeColor=<rgba 0 0.8594 1 1>}
  NamedRerouteDeclaration(None=@n1.Pre skinned Local Position) {name=PreskinLocPos, nodeColor=<rgba 1 0.8203 0 1>}
  NamedRerouteDeclaration(None=@n1.Tech Art Mask) {name=TechArtMask, nodeColor=<rgba 0.5 0 1 1>}
  n19 = NamedRerouteUsage()
  n20 = Add(A=@n12, B=Constant() {r=0.75})
  n21 = CurveAtlasRowParameter(CurveTime=ConstantBiasScale(None=Sine(None=@n20))) {
   atlas=/BRCosmetics/Gadgets/Assets/VinderTech_Umbrella/Umbrella_RankedRube/Textures/CA_RankedRube
   curve=/BRCosmetics/Gadgets/Assets/VinderTech_Umbrella/Umbrella_RankedRube/Textures/C_RankedRube_Emissive
   group=RankedRube
   parameterName=RadialColorCurve
  }
  n22 = NamedRerouteUsage()
  n23 = Multiply(A=Multiply(A=@n17, B=LinearInterpolate(A=@n19, B=@n21, Alpha=@n22))) {constB=2}
  n24 = OneMinus(None=NamedRerouteUsage())
  NamedRerouteDeclaration(None=LinearInterpolate(A=Multiply(A=@n23, B=@n24), B=@n6.Emissive Color, Alpha=NamedRerouteUsage())) {name=Emissive_OUT, nodeColor=<rgba 0 1 0.1797 1>}
  n25 = ComponentMask(None=NamedRerouteUsage()) {b=true}
  n26 = Step(X=@n11) {constX=0, constY=0.2}
  NamedRerouteDeclaration(None=Saturate(None=Multiply(A=@n25, B=@n26))) {name=SwirlingAlpha, nodeColor=<rgba 1 0.0234 0 1>}
 ]
}
